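; S3 Galileo wall poster building actor that stays tied to a calendar event for when it is shown.
; When the match is ready on authority, it watches the named calendar event and sets the replicated Visible flag once that event is active.
; The Visible rep-notify async-loads the Galileo material, applies it to slot 0 of the static mesh and unhides the actor.
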(event EventBeginPlay
  (Class|FortAsyncActionAuthorityMatchReady|MatchReadyAsync
    (:OnReady
      (MCPContext|WatchCalendarEvent (Variables|Default|GetCalendarEventName)
        (:OnEventActive
          (Variables|Default|SetVisible true))))))

(event Custom|SetMaterial
  (bind _object (Utilities|AsyncLoadAsset (Variables|Default|GetGalileoMaterial))
    (:Completed
      (bind _asmaterial_instance (Utilities|Casting|CastToMaterialInstance _object))
      (Rendering|Material|SetMaterial (Variables|Debug|GetStaticMeshComponent) 0 _asmaterial_instance)
      (Rendering|SetActorHiddenInGame false))))

(fn OnRep_Visible
  (CallFunction|SetMaterial))

(fn ConstructionScript)
